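# Material function that layers optional tiled detail textures over incoming material attributes, within a selected channel of a Tech Art Mask.
# Three static switches chain masked blends for SRM (specular/roughness/metallic), normal and diffuse.
# SRM is remapped to signed offsets scaled by strength, normals are flattened then angle-correct blended, and diffuse is lerped over base color.
# All detail samples share one UV set, switchable to UV2, with rotation and tiling.
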MaterialFunction /Game/Global/Materials/Character_Parent_2023/Detail_Texturing_Parent_2025/MF_Detail_Texturing_2025
graph {nodes [
  n1 = FunctionInput() {inputName=Result, inputType=FunctionInput_MaterialAttributes}
  n2 = GetMaterialAttributes(None=@n1) {attributeGetTypes [#1, #2, #3]}
  n3 = SetMaterialAttributes!(
   MaterialAttributes=@n2.MaterialAttributes
   Roughness=NamedRerouteUsage()
   Specular=NamedRerouteUsage()
   Metallic=NamedRerouteUsage()
   BaseColor
  ) {attributeSetTypes [
    #1
    #2
    #3
    #4
   ]}
  n4 = ChannelMaskParameter(None=FunctionInput() {inputName=Tech Art Mask, sortPriority=10}) {group=Detail Texturing, parameterName=Detail Texture - Channel Mask, sortPriority=-2}
  n5 = StaticSwitchParameter(True=BlendMaterialAttributes(A=@n1, B=@n3, Alpha=@n4.R), False=@n1) {group=Detail Texturing, parameterName=Use Detail SRM?, sortPriority=-3}
  n6 = GetMaterialAttributes(None=@n5) {attributeGetTypes [#5]}
  n7 = MaterialFunctionCall(BaseNormal=@n6.Normal, AdditionalNormal=NamedRerouteUsage()) {materialFunction=&BlendAngleCorrectedNormals}
  n8 = StaticSwitchParameter(
   True=BlendMaterialAttributes(
    A=@n5
    B=SetMaterialAttributes!(MaterialAttributes=@n6.MaterialAttributes, Normal=@n7.Result) {attributeSetTypes [#5]}
    Alpha=@n4.R
   )
   False=@n5
  ) {group=Detail Texturing, parameterName=Use Detail Normal?, sortPriority=-3}
  n9 = GetMaterialAttributes(None=@n8) {attributeGetTypes [#4]}
  n10 = SetMaterialAttributes!(MaterialAttributes=@n9.MaterialAttributes, BaseColor=NamedRerouteUsage()) {attributeSetTypes [#4]}
  FunctionOutput(None=StaticSwitchParameter(True=BlendMaterialAttributes(A=@n8, B=@n10, Alpha=@n4.R), False=@n8) {group=Detail Texturing, parameterName=Use Detail Diffuse?, sortPriority=-5}) {outputName=Material Attributes}
  n11 = TextureCoordinate() {coordinateIndex=1}
  n12 = TextureCoordinate()
  n13 = Multiply(
   A=MaterialFunctionCall(
    UVs=StaticSwitchParameter(True=@n11, False=@n12) {group=Detail Texturing, parameterName=Detail Texture - Use UV2, sortPriority=-2}
    "Rotation Angle (0-1)"=ScalarParameter() {group=Detail Texturing, parameterName=Detail Texture - UV Rotation, sortPriority=0}
   ) {materialFunction=&CustomRotator}.Rotated Values
   B=ScalarParameter() {defaultValue=1, group=Detail Texturing, parameterName=Detail Texture - Tiling, sortPriority=-1}
  )
  n14 = TextureSampleParameter2D(UVs=@n13) {
   desc="R = Spec , G = Roughness, B= Metal"
   group=Detail Texturing
   parameterName=Detail SRM
   samplerType=SAMPLERTYPE_Masks
   sortPriority=3
   texture=&T_Fortnite_Default_S
  }
  n15 = MaterialFunctionCall(Float3=ConstantBiasScale(None=@n14.RGB) {bias=-0.5, scale=2}) {materialFunction=&BreakOutFloat3Components}
  n16 = ScalarParameter() {defaultValue=0.1, group=Detail Texturing, parameterName=Detail Roughness - Strength, sortPriority=8}
  NamedRerouteDeclaration(None=Add(A=Multiply(A=@n15.G, B=@n16), B=NamedRerouteUsage())) {name=Roughness, nodeColor=<rgba 1 0 0.8203 1>}
  n17 = ScalarParameter() {defaultValue=0.1, group=Detail Texturing, parameterName=Detail Metallic - Strength, sortPriority=9}
  NamedRerouteDeclaration(None=Add(A=Multiply(A=@n15.B, B=@n17), B=NamedRerouteUsage())) {name=Metallic, nodeColor=<rgba 0 1 0.8594 1>}
  n18 = ScalarParameter() {defaultValue=0.1, group=Detail Texturing, parameterName=Detail Specular - Strength, sortPriority=7}
  NamedRerouteDeclaration(None=Add(A=Multiply(A=@n15.R, B=@n18), B=NamedRerouteUsage())) {name=Specular, nodeColor=<rgba 1 0.5391 0 1>}
  n19 = TextureSampleParameter2D(UVs=@n13) {
   group=Detail Texturing
   parameterName=Detail Normal
   samplerType=SAMPLERTYPE_Normal
   sortPriority=2
   texture=&FlatNormalBlue
  }
  NamedRerouteDeclaration(None=MaterialFunctionCall(
    Normal=@n19.RGB
    Flatness=ScalarParameter() {defaultValue=0.1, group=Detail Texturing, parameterName=Detail Normal - Flatten Normal, sliderMax=1, sortPriority=3}
   ) {materialFunction=&FlattenNormal}.Result) {name=Detail Normal, nodeColor=<rgba 0 0.7422 1 1>}
  NamedRerouteDeclaration(None=@n2.Roughness) {name=Roughness In, nodeColor=<rgba 1 0 0.8203 1>}
  NamedRerouteDeclaration(None=@n2.Specular) {name=Specular In, nodeColor=<rgba 1 0.5391 0 1>}
  NamedRerouteDeclaration(None=@n2.Metallic) {name=Metallic In, nodeColor=<rgba 0 1 0.8594 1>}
  NamedRerouteDeclaration(None=@n9.BaseColor) {name=Base Color In, nodeColor=<rgba 1 0.3047 0 1>}
  n20 = TextureSampleParameter2D(UVs=@n13) {group=Detail Texturing, parameterName=Detail Diffuse, sortPriority=1, texture=&T_BlacksRGB}
  NamedRerouteDeclaration(None=LinearInterpolate(
    A=NamedRerouteUsage()
    B=@n20.RGB
    Alpha=ScalarParameter() {defaultValue=0.1, group=Detail Texturing, parameterName=Detail Diffuse - Strength, sortPriority=2}
   )) {name=Base Color, nodeColor=<rgba 0 0.0156 1 1>}
  NamedRerouteUsage()
 ]}
